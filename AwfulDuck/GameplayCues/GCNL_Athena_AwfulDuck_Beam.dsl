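; Looping gameplay cue for the AwfulDuck beam, a spotlight effect aimed from the cue location at the target player.
; A repeating timer points the Niagara beam and moves its audio to the pawn's Spine_05 socket.
; Local players also get a radial force field and timeline-driven white-out with camera shake through a material parameter collection.
; Beam audio fades into 1P/3P exit sounds after a delay; removal clears timer and post-process.
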(event EventOnLoopingStart(EitherVFX-preferthisversion)
  (params (MyTarget :type Actor%20Object%20Reference) (Parameters :type Gameplay%20Cue%20Parameters%20Structure%20%28by%20ref%29) (ParticleComponents :type Array%20of%20FXSystem%20Component%20Object%20References) (AudioComponents :type Array%20of%20Audio%20Component%20Object%20References))
  (bind (_normalizedmagnitude _rawmagnitude _effectcontext _matchedtagname _originaltag _aggregatedsourcetags _aggregatedtargettags _location) (Ability|GameplayCue|BreakGameplayCueParameters Parameters))
  (bind _islocallycontrolled (Variables|Default|GetIsLocallyControlled))
  (bind _beamfx_component (Variables|Default|GetBeamFX_Component))
  (bind _postprocess_whiteout (Variables|Default|GetPostProcess_WhiteOut))
  (Development|PrintString 0 "GCNL BEAM - Start")
  (Utilities|FlowControl|Sequence
    (:then_0
      (bind _result (ArrayMacros|SafeArrayGet ParticleComponents)
        (:%20%20
          (Variables|Default|SetBeamOrigin _location)
          (bind _output_get (Variables|Default|SetTargetPawnActor MyTarget))
          (bind _asplayer_pawn_athena (Utilities|Casting|CastToPlayerPawn_Athena _output_get))
          (Variables|Default|SetTargetPawn _asplayer_pawn_athena)
          (Variables|Default|SetTargetPawnMeshComp (Variables|/Script/Engine.Character|GetMesh _asplayer_pawn_athena))
          (Variables|Default|SetIsLocallyControlled (Pawn|IsLocallyControlled (Variables|Default|GetTargetPawn)))
          (bind _returnvalue (Utilities|Time|SetTimerbyEvent (delegate Custom|CE_UpdateBeamTarget) 0.02 true))
          (Variables|Default|SetTimer _returnvalue)
          (Variables|Default|SetBeamFX_Component _result)
          (Transformation|SetWorldLocation _beamfx_component (Variables|Default|GetBeamOrigin) false false)
          (Variables|Default|SetBeamNiagaraSys (Utilities|Casting|CastToNiagaraComponent _beamfx_component))
          (Rendering|SetTranslucentSortPriority _beamfx_component 3))))
    (:then_1
      (if (Variables|Default|GetIsLocallyControlled)
        (bind _targetpawnactor (Variables|Default|GetTargetPawnActor)
          (:then
            (Game|SpawnActorfromClass 0 "/Game/Effects/Niagara/Blueprints/PhsyicsFields/FN_RadialForce.FN_RadialForce_C" (Transformation|GetActorTransform _targetpawnactor) "Undefined" "MultiplyWithRoot" 0 0.2 5.0 3.5e+03 0.4 0.0 0))
          (:else))))
    (:then_2
      (if (Variables|Default|GetIsLocallyControlled)
        (bind _targetpawnmeshcomp (Variables|Default|GetTargetPawnMeshComp)
          (:then
            (Transformation|AttachComponentToComponent _postprocess_whiteout _targetpawnmeshcomp "root" "KeepRelative" "KeepRelative" "KeepRelative" true)
            (Variables|/Script/Engine.PostProcessComponent|SetbEnabled true _postprocess_whiteout)
            (bind (_direction _pp_blendweight _pp_camerashake) (Timeline|TFX_WhiteOut PlayFromStart)
              (:Update
                (Variables|/Script/Engine.PostProcessComponent|SetBlendWeight _pp_blendweight (Variables|Default|GetPostProcess_WhiteOut))
                (Rendering|Material|SetScalarParameterValue 0 "/Game/Athena/Prototype/Blueprints/Galileo/Blueprints/MPC_Galileo.MPC_Galileo" "CameraShake" _pp_camerashake :!class /Script/Engine.KismetMaterialLibrary)
                (Utilities|FlowControl|DoOnce false
                  (:Completed
                    (Audio|SpawnSoundatLocation (Variables|Default|GetTargetPawn) "/AwfulDuck/Audio/AwfulDuck_Spotlight_Flash_Cue.AwfulDuck_Spotlight_Flash_Cue" (Variables|Default|GetBeamOrigin)))))))
          (:else))))
    (:then_3
      (bind _returnvalue_2 (Audio|SpawnSoundatLocation 0 (Variables|Default|GetBeamPlayerSound) (Transformation|GetActorLocation (Variables|Default|GetTargetPawn))))
      (bind _output_get_2 (Variables|Default|SetBeamPlayerAudio _returnvalue_2))
      (Audio|Parameter|SetBooleanParameter _output_get_2 "IsLocal" (Variables|Default|GetIsLocallyControlled) :!class /Script/AudioExtensions.AudioParameterControllerInterface)
      (Utilities|FlowControl|Delay _output_get_2 3.3)
      (bind _beamplayeraudio (Variables|Default|GetBeamPlayerAudio)
        (:then
          (Audio|Components|Audio|FadeOut _beamplayeraudio (Math|Float|SelectFloat 2.0 0.5 _islocallycontrolled) 0.0 "Linear" :!class /Script/Engine.AudioComponent)
          (Audio|PlaySoundatLocation _output_get_2 (select _islocallycontrolled (Variables|Default|GetSpotlightPlayerExit1P) (Variables|Default|GetSpotlightPlayerExit3P)) (Variables|Default|GetTargetLocation)))
        (:else)))))

(event Custom|CE_UpdateBeamTarget
  (Variables|Default|SetTargetLocation (Transformation|GetSocketLocation (Variables|Default|GetTargetPawnMeshComp) "Spine_05"))
  (Transformation|SetWorldRotation (Variables|Default|GetBeamFX_Component) (Math|Rotator|FindLookatRotation (Variables|Default|GetBeamOrigin) (Variables|Default|GetTargetLocation)) false false)
  (Niagara|SetNiagaraVariable(Vector3) (Variables|Default|GetBeamNiagaraSys) "BeamTarget" (Variables|Default|GetTargetLocation))
  (bind _beamplayeraudio (Variables|Default|GetBeamPlayerAudio)
    (:then
      (Transformation|SetWorldLocation _beamplayeraudio (Variables|Default|GetTargetLocation) false false))
    (:else)))

(event EventOnRemoval(EitherVFX-preferthisversion)
  (params (MyTarget :type Actor%20Object%20Reference) (Parameters :type Gameplay%20Cue%20Parameters%20Structure%20%28by%20ref%29) (ParticleComponents :type Array%20of%20FXSystem%20Component%20Object%20References) (AudioComponents :type Array%20of%20Audio%20Component%20Object%20References) (BurstCameraShakeInstance :type Camera%20Shake%20Base%20Object%20Reference) (BurstDecalInstance :type Decal%20Actor%20Object%20Reference))
  (bind _postprocess_whiteout (Variables|Default|GetPostProcess_WhiteOut))
  (Development|PrintString 0 "GCNL BEAM - End")
  (Utilities|Time|ClearandInvalidateTimerbyHandle 0 (Variables|Default|GetTimer))
  (if (Variables|/Script/Engine.PostProcessComponent|GetbEnabled _postprocess_whiteout)
    (Variables|/Script/Engine.PostProcessComponent|SetbEnabled false _postprocess_whiteout)
    (Variables|/Script/Engine.PostProcessComponent|SetBlendWeight 0.0 _postprocess_whiteout)
    (Rendering|Material|SetScalarParameterValue 0 "/Game/Athena/Prototype/Blueprints/Galileo/Blueprints/MPC_Galileo.MPC_Galileo" "CameraShake" :!class /Script/Engine.KismetMaterialLibrary)))

(fn RestoreCharacterMaterials
  (bind _currentreplacementindex (Variables|Default|GetCurrentReplacementIndex))
  (bind _targetpawn (Variables|Default|GetTargetPawn))
  (bind _pawnmaterials_all (Variables|/Game/Athena/PlayerPawn_Athena_Generic_Parent.PlayerPawn_Athena_Generic_Parent_C|GetPawnMaterials_ALL _targetpawn))
  (bind _returnvalue (Utilities|Array|IsValidIndex _pawnmaterials_all _currentreplacementindex))
  (bind _output (Utilities|Array|Get(acopy) _pawnmaterials_all _currentreplacementindex))
  (bind _returnvalue_2 (select _returnvalue _output 0))
  (TeleportationFX|CleanupDissolveFX (Variables|Default|GetTargetPawn) (Variables|Default|GetMat%20Character%20Dissolve) (Variables|Default|GetDissolveMIDs))
  (for _array_element (Variables|Default|GetDissolveMeshParts)
    (for (_array_element_2 _array_index) (Rendering|Material|GetMaterials _array_element)
      (Variables|Default|SetCurrentReplacementIndex (+ _array_index (Variables|Default|GetItr)))
      (Rendering|Material|SetMaterial _array_element _array_index _returnvalue_2)
      (Rendering|Material|SetScalarParameterValue _returnvalue_2 "NewDissolveGradient" 0.0 :!class /Script/Engine.MaterialInstanceDynamic))
    (Variables|Default|SetItr (+ (Variables|Default|GetItr) (Utilities|Array|Length (Rendering|Material|GetMaterials _array_element)))))
  (return true))

(fn ConstructionScript)
