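; Looping GameplayCue for the AwfulDuck rez out-and-in effect, dissolving a player pawn with a skin-swap effect.
; On start, it applies a dissolve material override smeared toward the cue location, starts a timeline, spawns chest FX, and hides the pet and non-abduction FX.
; On removal, it restores body and weapon materials with delayed retries, reapplies the weapon wrap and unhides the pawn before calling parent removal.
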(event EventOnLoopingStart(EitherVFX-preferthisversion)
  (params (MyTarget :type Actor%20Object%20Reference) (Parameters :type Gameplay%20Cue%20Parameters%20Structure%20%28by%20ref%29) (ParticleComponents :type Array%20of%20FXSystem%20Component%20Object%20References) (AudioComponents :type Array%20of%20Audio%20Component%20Object%20References))
  (bind (_normalizedmagnitude _rawmagnitude _effectcontext _matchedtagname _originaltag _aggregatedsourcetags _aggregatedtargettags _location) (Ability|GameplayCue|BreakGameplayCueParameters Parameters))
  (Variables|Default|SetBeamOrigin _location)
  (Utilities|FlowControl|Sequence
    (:then_0
      (bind _asplayer_pawn_athena (Utilities|Casting|CastToPlayerPawn_Athena MyTarget))
      (Variables|Default|SetTargetPawn _asplayer_pawn_athena)
      (Variables|Default|SetTargetLocation (Transformation|GetActorLocation (Variables|Default|GetTargetPawn))))
    (:then_1
      (bind _returnvalue (Class|FortPlayerPawn|ApplyMaterialOverrideLocalwithFlags (Variables|Default|GetTargetPawn) (Utilities|ResolveSoftReference (Variables|Default|GetMat%20Character%20Dissolve)) (Variables|Default|GetMaterial%20Copy%20Params) 0.0 0))
      (Variables|Default|SetMaterial%20Override%20ID _returnvalue)
      (Class|FortPlayerPawn|SetScalarParamonOverriddenMaterials (Variables|Default|GetTargetPawn) "DissolveGradientPercentageNone" 1.0)
      (Class|FortPlayerPawn|SetVectorParamonOverriddenMaterials (Variables|Default|GetTargetPawn) "ExplodeSmear" (Math|Conversions|ToLinearColor(Vector) (* (Math|Vector|Normalize (- (Variables|Default|GetBeamOrigin) (Variables|Default|GetTargetLocation))) 1e+02)))
      (Rendering|SetActorHiddenInGame (Variables|/Script/FortniteGame.FortPlayerPawn|GetCosmeticPetInstance (Variables|Default|GetTargetPawn)) true)
      (Components|Timeline|PlayfromStart (Variables|GCNL_Athena_AwfulDuck_RezOutAndIn|GetTFX_ResOutCharacterMesh))
      (Effects|SpawnFXSystemAttached :FXSystemAsset (Variables|Default|GetFX_SkinSwap) :AttachToComponent (Variables|/Script/Engine.Character|GetMesh (Variables|Default|GetTargetPawn)) :AttachPointName "FX_Chest" :Scale "1.000000,1.000000,1.000000" :LocationType "SnapToTarget" :bAutoDestroy true :PoolingMethod "None" :bAutoActivate true)
      (for _array_element (Components|GetChildrenComponents (Variables|/Script/Engine.Character|GetMesh (Variables|Default|GetTargetPawn)) true)
        (bind _asfxsystem_component (Utilities|Casting|CastToFXSystemComponent _array_element))
        (if (not (Utilities|Array|ContainsItem (Variables|Default|GetAbductionFXAssets) (Effects|Components|ParticleSystem|GetFXSystemAsset _asfxsystem_component)))
          (Development|SetHiddeninGame _asfxsystem_component true true)
          (if (Variables|Default|GetDEBUG_FXSYSTEMCOMPONENTS)
            (Development|PrintString 0 (Utilities|GetDisplayName _asfxsystem_component) true true "(R=0.000000,G=1.000000,B=0.203936,A=1.000000)" 6e+01)))))))

(event Custom|CE_SetPawnHidden
  (bind _new_hidden (Variables|Default|GetNew%20Hidden))
  (bind _targetpawn (Variables|Default|GetTargetPawn))
  (bind _new_hidden_2 (Variables|Default|GetNew%20Hidden))
  (Rendering|SetActorHiddenInGame (Variables|/Script/Engine.Pawn|GetPlayerState _targetpawn) _new_hidden)
  (Class|PlayerPawnAthena|HideCharacterParts _targetpawn _new_hidden)
  (for _array_element (Components|GetChildrenComponents (Variables|/Script/Engine.Character|GetMesh (Variables|Default|GetTargetPawn)) true)
    (bind _asfxsystem_component (Utilities|Casting|CastToFXSystemComponent _array_element))
    (Development|SetHiddeninGame _asfxsystem_component _new_hidden_2 true))
  (Rendering|SetActorHiddenInGame (Variables|/Script/FortniteGame.FortPlayerPawn|GetCosmeticPetInstance (Variables|Default|GetTargetPawn)) _new_hidden_2))

(event EventOnRemoval(EitherVFX-preferthisversion)
  (params (MyTarget :type Actor%20Object%20Reference) (Parameters :type Gameplay%20Cue%20Parameters%20Structure%20%28by%20ref%29) (ParticleComponents :type Array%20of%20FXSystem%20Component%20Object%20References) (AudioComponents :type Array%20of%20Audio%20Component%20Object%20References) (BurstCameraShakeInstance :type Camera%20Shake%20Base%20Object%20Reference) (BurstDecalInstance :type Decal%20Actor%20Object%20Reference))
  (bind _targetpawn (Variables|Default|GetTargetPawn))
  (bind _returnvalue (CosmeticLoadout|GetItemWrapsfromActor _targetpawn))
  (bind _returnvalue_2 (Data|GetWeaponData (Variables|/Script/FortniteGame.FortPawn|GetCurrentWeapon (Variables|Default|GetTargetPawn))))
  (bind _targetpawn_2 (Variables|Default|GetTargetPawn))
  (bind _currentweapon (Variables|/Script/FortniteGame.FortPawn|GetCurrentWeapon _targetpawn_2))
  ; Fully Restore Mats Over Time_2
  (bind _fully_completed (Default|RestoreCharacterMaterials))
  (bind _result (Math|Integer|DecrementInt (Variables|Default|GetAmountOfTimesToAttemptRestoreMats))
    (:%20%20
      (if (not (or _fully_completed (== _result 0)))
        (Utilities|FlowControl|Delay 0))))
  (Utilities|IsValid _targetpawn_2
    (:Is%20Valid
      (Utilities|IsValid _currentweapon
        (:Is%20Valid
          (Utilities|IsValid _returnvalue_2
            (:Is%20Valid
              (bind _returnvalue_3 (Cosmetic|GetItemWrapperLockerSlotIndex _returnvalue_2))
              (if (Utilities|Array|IsValidIndex _returnvalue _returnvalue_3)
                (Cosmetic|ApplyItemWraptoActor (Utilities|Array|Get(acopy) _returnvalue _returnvalue_3) _currentweapon)
                (else
                  _))
              (Variables|Default|SetNew%20Hidden false)
              (CallFunction|CESetPawnHidden)
              (CallFunction|CallParentRemoval))
            (:Is%20Not%20Valid
              _)))
        (:Is%20Not%20Valid
          _)))
    (:Is%20Not%20Valid
      _)))

(event Custom|Call%20Parent%20Removal
  (|CallParentFunction (stale EventOnRemoval(EitherVFX-preferthisversion) MyTarget) (stale EventOnRemoval(EitherVFX-preferthisversion) Parameters) (stale EventOnRemoval(EitherVFX-preferthisversion) ParticleComponents) (stale EventOnRemoval(EitherVFX-preferthisversion) AudioComponents) (stale EventOnRemoval(EitherVFX-preferthisversion) BurstCameraShakeInstance) (stale EventOnRemoval(EitherVFX-preferthisversion) BurstDecalInstance)))

(fn GetDissolveMeshParts
  (bind _targetpawn (Variables|Default|GetTargetPawn))
  (Utilities|Set|Clear (LocalVariables|GetMeshSet))
  (Utilities|IsValid _targetpawn
    (:Is%20Valid
      (Utilities|Set|AddItems (LocalVariables|GetMeshSet) (Variables|/Game/Athena/PlayerPawn_Athena_Generic_Parent.PlayerPawn_Athena_Generic_Parent_C|GetSkeletalMeshes _targetpawn))
      (Utilities|Set|Add (LocalVariables|GetMeshSet) (Default|GetWeaponMesh))
      (bind _result (Utilities|Set|ToArray (LocalVariables|GetMeshSet)))
      (return _result))
    (:Is%20Not%20Valid
      (return 0))))

(fn GetWeaponMesh
  (bind _targetpawn (Variables|Default|GetTargetPawn))
  (bind _currentweapon (Variables|/Script/FortniteGame.FortPawn|GetCurrentWeapon _targetpawn))
  (Utilities|IsValid _targetpawn
    (:Is%20Valid
      (Utilities|IsValid _currentweapon
        (:Is%20Valid
          (return (Variables|/Script/FortniteGame.FortWeapon|GetWeaponMesh _currentweapon)))
        (:Is%20Not%20Valid
          _)))
    (:Is%20Not%20Valid
      _))
  (return 0))

(fn RestoreCharacterMaterials
  (bind _currentreplacementindex (Variables|Default|GetCurrentReplacementIndex))
  (bind _targetpawn (Variables|Default|GetTargetPawn))
  (bind _pawnmaterials_all (Variables|/Game/Athena/PlayerPawn_Athena_Generic_Parent.PlayerPawn_Athena_Generic_Parent_C|GetPawnMaterials_ALL _targetpawn))
  (bind _returnvalue (Utilities|Array|IsValidIndex _pawnmaterials_all _currentreplacementindex))
  (bind _output (Utilities|Array|Get(acopy) _pawnmaterials_all _currentreplacementindex))
  (bind _returnvalue_2 (select _returnvalue _output 0))
  (for _array_element (CallFunction|GetDissolveMeshParts)
    (Utilities|IsValid _array_element
      (:Is%20Valid
        (for (_ _array_index) (Rendering|Material|GetMaterials _array_element)
          (Variables|Default|SetCurrentReplacementIndex (+ _array_index (Variables|Default|GetItr)))
          (Rendering|Material|SetMaterial _array_element _array_index _returnvalue_2)
          (Rendering|Material|SetScalarParameterValue _returnvalue_2 "NewDissolveGradient" 0.0 :!class /Script/Engine.MaterialInstanceDynamic))
        (Variables|Default|SetItr (+ (Variables|Default|GetItr) (Utilities|Array|Length (Rendering|Material|GetMaterials _array_element)))))
      (:Is%20Not%20Valid
        (Development|PrintString 0 "SomeDissolveMeshPartsAreInvalid!" true true "(R=1.000000,G=0.015653,B=0.000000,A=1.000000)" 1e+01))))
  (return true))

(fn ConstructionScript)
