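; Looping gameplay cue notify for the AwfulDuck warm-up, containing logic to aim a warning effect at a target pawn.
; CE_UpdateWarningFX stores the target mesh's Spine_05 socket location and rotates the warning FX from its origin toward it.
; It also prints the scaled origin-to-target distance and feeds it to the Niagara X_Length variable to stretch the effect.
; Loop start and removal only print debug strings; nothing shown here calls the update event.
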(event EventOnLoopingStart(EitherVFX-preferthisversion)
  (params (MyTarget :type Actor%20Object%20Reference) (Parameters :type Gameplay%20Cue%20Parameters%20Structure%20%28by%20ref%29) (ParticleComponents :type Array%20of%20FXSystem%20Component%20Object%20References) (AudioComponents :type Array%20of%20Audio%20Component%20Object%20References))
  (Development|PrintString 0 "GCNL WarmUp - Start"))

(event Custom|CE_UpdateWarningFX
  (bind _warningeffect_target_location (Variables|Default|GetWarningEffect_Target_Location))
  (bind _warningeffect_origion_location (Variables|Default|GetWarningEffect_Origion_Location))
  (bind _returnvalue (Math|Vector|Distance(Vector) _warningeffect_origion_location _warningeffect_target_location))
  (bind _returnvalue_2 (Math|Float|SafeDivide _returnvalue 1e+04))
  (bind _targetpawnmeshcomp (Variables|Default|GetTargetPawnMeshComp)
    (:then
      (Variables|Default|SetWarningEffect_Target_Location (Transformation|GetSocketLocation _targetpawnmeshcomp "Spine_05"))
      (bind _warningfx (Variables|Default|GetWarningFX)
        (:then
          (Transformation|SetWorldRotation _warningfx (Math|Rotator|FindLookatRotation (Variables|Default|GetWarningEffect_Origion_Location) (Variables|Default|GetWarningEffect_Target_Location)) false false)
          (Development|PrintString 0 (Utilities|String|ToString(Float) _returnvalue_2) false false)
          (Niagara|SetNiagaraVariable(Float) (Variables|Default|GetWarningFXNiagaraComp) "X_Length" _returnvalue_2))
        (:else)))
    (:else)))

(event EventOnRemoval(EitherVFX-preferthisversion)
  (params (MyTarget :type Actor%20Object%20Reference) (Parameters :type Gameplay%20Cue%20Parameters%20Structure%20%28by%20ref%29) (ParticleComponents :type Array%20of%20FXSystem%20Component%20Object%20References) (AudioComponents :type Array%20of%20Audio%20Component%20Object%20References) (BurstCameraShakeInstance :type Camera%20Shake%20Base%20Object%20Reference) (BurstDecalInstance :type Decal%20Actor%20Object%20Reference))
  (Development|PrintString 0 "GCNL WarmUp - End"))

(fn ConstructionScript)
