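; Looping gameplay cue for the AwfulDuck abduction warning, showing a beam and warning effect on the targeted pawn.
; On start it takes the origin from the instigator or cue location, aims the mesh-attached beam there and starts spotlight and player audio loops, or hides the effects without an origin.
; Locally controlled targets get a wash-force field actor when physics fields are enabled.
; Removal deactivates the beam, fades spotlight audio and destroys the field.
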(event EventOnLoopingStart(EitherVFX-preferthisversion)
  (params (MyTarget :type Actor%20Object%20Reference) (Parameters :type Gameplay%20Cue%20Parameters%20Structure%20%28by%20ref%29) (ParticleComponents :type Array%20of%20FXSystem%20Component%20Object%20References) (AudioComponents :type Array%20of%20Audio%20Component%20Object%20References))
  (bind _returnvalue (Ability|EffectContext|GetInstigatorActor (.EffectContext (Ability|GameplayCue|BreakGameplayCueParameters Parameters))))
  (bind (_normalizedmagnitude _rawmagnitude _effectcontext _matchedtagname _originaltag _aggregatedsourcetags _aggregatedtargettags _location) (Ability|GameplayCue|BreakGameplayCueParameters Parameters))
  (bind _awfulduck_warningbeam (Variables|Default|GetAwfulDuck_WarningBeam))
  (bind _warningfxsystemcomponent (Variables|FortGameplayCueNotifyLoop_AbductionWarning|GetWarningFXSystemComponent))
  (bind _targetpawn (Variables|Default|GetTargetPawn))
  (bind _awfulduck_warningbeam_2 (Variables|Default|GetAwfulDuck_WarningBeam))
  (bind _mesh (Variables|/Script/Engine.Character|GetMesh _targetpawn))
  (if (ScalableFloat|GetValueasBool (Variables|Default|GetUseLocalSpots))
    (bind _asfort_pawn (Utilities|Casting|CastToFortPawn MyTarget))
    (if (or (Pawn|IsLocallyViewed _asfort_pawn) (Pawn|IsLocallyControlled _asfort_pawn))
      _)
    (else
      _))
  (bind _result (ArrayMacros|SafeArrayGet ParticleComponents)
    (:%20%20
      (bind _result_2 (ArrayMacros|SafeArrayGet ParticleComponents 1)
        (:%20%20
          (if (or (Utilities|IsValid (Ability|EffectContext|GetInstigatorActor _effectcontext) :!class /Script/Engine.KismetSystemLibrary) (not (Math|Vector|VectorIsNearlyZero _location)))
            (bind _asniagara_particle_system_component (Utilities|Casting|CastToNiagaraComponent _result_2))
            (Variables|Default|SetWarmUpRingFX _asniagara_particle_system_component)
            (Rendering|SetVisibility _awfulduck_warningbeam true false)
            (Components|Activation|Activate _awfulduck_warningbeam false)
            (Variables|FortGameplayCueNotifyLoop_AbductionWarning|SetWarningFXSystemComponent _result)
            (bind _output_get (Variables|FortGameplayCueNotifyLoop_AbductionWarning|SetTargetActor MyTarget))
            (if (Utilities|IsValid _returnvalue :!class /Script/Engine.KismetSystemLibrary)
              (Variables|FortGameplayCueNotifyLoop_AbductionWarning|SetWarningFXOriginLocation (Transformation|GetActorLocation _returnvalue))
              (else
                (Variables|FortGameplayCueNotifyLoop_AbductionWarning|SetWarningFXOriginLocation _location)))
            ;[0.9525 1 0] Audio
            (bind _returnvalue_2 (Audio|SpawnSoundatLocation _output_get (Variables|Default|GetSpotlightLoop) (Variables|FortGameplayCueNotifyLoop_AbductionWarning|GetWarningFXOriginLocation)))
            (Variables|Default|SetWarningAudio _returnvalue_2)
            (bind _returnvalue_3 (Audio|SpawnSoundAttached (Variables|Default|GetLocalPlayerLoop) (Variables|/Script/Engine.Actor|GetRootComponent _output_get)))
            (bind _output_get_2 (Variables|FortGameplayCueNotifyLoop_AbductionWarning|SetWarningAudioComponent _returnvalue_3))
            (Audio|Parameter|SetBooleanParameter _output_get_2 "IsLocal" (Pawn|IsLocallyViewed (Utilities|Casting|CastToFortPawn _output_get)) :!class /Script/AudioExtensions.AudioParameterControllerInterface)
            (Utilities|FlowControl|Sequence
              (:then_0
                (bind _asfort_player_pawn (Utilities|Casting|CastToFortPlayerPawn (Variables|FortGameplayCueNotifyLoop_AbductionWarning|GetTargetActor)))
                (Variables|Default|SetTargetPawn _asfort_player_pawn)
                (Rendering|SetTranslucentSortPriority _warningfxsystemcomponent 1)
                (Rendering|SetVisibility _warningfxsystemcomponent true false) ; Make sure its visible
                (bind _asniagara_particle_system_component_2 (Utilities|Casting|CastToNiagaraComponent _warningfxsystemcomponent))
                (bind _output_get_3 (Variables|FortGameplayCueNotifyLoop_AbductionWarning|SetWarningFXNiagaraComponent _asniagara_particle_system_component_2))
                (Components|Activation|Activate _output_get_3 false) ; Make sure its active
                (Transformation|SetWorldLocation (Variables|FortGameplayCueNotifyLoop_AbductionWarning|GetWarningFXSystemComponent) (Variables|FortGameplayCueNotifyLoop_AbductionWarning|GetWarningFXOriginLocation) false false)
                (bind _returnvalue_4 (Development|GetConsoleVariableIntValue "r.PhysicsField.EnableField"))
                (if (and (== _returnvalue_4 1) (Pawn|IsLocallyControlled (Variables|Default|GetTargetPawn)))
                  (bind _targetactor (Variables|FortGameplayCueNotifyLoop_AbductionWarning|GetTargetActor)
                    (:then
                      (Variables|FortGameplayCueNotifyLoop_AbductionWarning|GetRadialForceLoop
                        (:then)
                        (:else
                          (bind _returnvalue_5 (Game|SpawnActorfromClass _targetactor "/AwfulDuck/Athena/Items/AwfulDuck/FX/FN_AwfulDuck_WashForce.FN_AwfulDuck_WashForce_C" (Transformation|GetActorTransform _targetactor) "Undefined" "MultiplyWithRoot" 0 4.35e+03 0.65 -999.0 0))
                          (Variables|FortGameplayCueNotifyLoop_AbductionWarning|SetRadialForceLoop _returnvalue_5))))
                    (:else))))
              (:then_1
                (Variables|Default|SetTargetPawnMeshComp _mesh)
                (Transformation|AttachComponentToComponent (Variables|Default|GetAwfulDuck_WarningBeam) _mesh "root" "SnapToTarget" "SnapToTarget" "KeepRelative" true)
                (Effects|Components|ParticleSystem|SetVectorParameter (Variables|Default|GetAwfulDuck_WarningBeam) "BeamTarget" (Variables|FortGameplayCueNotifyLoop_AbductionWarning|GetWarningFXOriginLocation))))
            (else
              (Development|PrintString 0 "Cannot determine a location for the beam, hiding the beam")
              ;[1 0.4192 0.3694] Hide and deactivate the fx that would use WarningFXOrigin Location
              (Components|Activation|Deactivate _awfulduck_warningbeam_2)
              (Rendering|SetVisibility _awfulduck_warningbeam_2 false false)
              (Rendering|SetVisibility _result false false)
              (Components|Activation|Deactivate _result))))))))

(event EventOnRemoval(EitherVFX-preferthisversion)
  (params (MyTarget :type Actor%20Object%20Reference) (Parameters :type Gameplay%20Cue%20Parameters%20Structure%20%28by%20ref%29) (ParticleComponents :type Array%20of%20FXSystem%20Component%20Object%20References) (AudioComponents :type Array%20of%20Audio%20Component%20Object%20References) (BurstCameraShakeInstance :type Camera%20Shake%20Base%20Object%20Reference) (BurstDecalInstance :type Decal%20Actor%20Object%20Reference))
  (Development|PrintString 0 "GCNL Warning - End")
  (Development|PrintString 0 "WarningGC_GetsRemoved" false false "(R=0.086353,G=1.000000,B=0.040309,A=1.000000)" 1e+01)
  (Components|Activation|Deactivate (Variables|Default|GetAwfulDuck_WarningBeam))
  (Audio|Components|Audio|FadeOut (Variables|Default|GetWarningAudio) 1.0 0.0 "Linear" :!class /Script/Engine.AudioComponent)
  (bind _radialforceloop (Variables|FortGameplayCueNotifyLoop_AbductionWarning|GetRadialForceLoop)
    (:then
      (Actor|DestroyActor _radialforceloop))
    (:else)))

(fn ConstructionScript)
