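; Looping gameplay cue for the AwfulDuck warning (LocalLight variant) that keeps a dirt particle effect on the target player pawn.
; On loop start it caches the pawn, the first particle component and the source duck's root location, then starts a repeating timer.
; Each tick moves the effect to the pawn's mesh and points its initial velocity parameter toward that cached location.
; Removal clears the timer; both events print debug strings.
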(event EventOnLoopingStart(EitherVFX-preferthisversion)
  (params (MyTarget :type Actor%20Object%20Reference) (Parameters :type Gameplay%20Cue%20Parameters%20Structure%20%28by%20ref%29) (ParticleComponents :type Array%20of%20FXSystem%20Component%20Object%20References) (AudioComponents :type Array%20of%20Audio%20Component%20Object%20References))
  (bind (_normalizedmagnitude _rawmagnitude _effectcontext _matchedtagname _originaltag _aggregatedsourcetags _aggregatedtargettags _location _normal _instigator _effectcauser _sourceobject) (Ability|GameplayCue|BreakGameplayCueParameters Parameters))
  (Development|PrintString 0 "GCNL LocalLight - Start")
  (bind _asfort_player_pawn_athena (Utilities|Casting|CastToFortPlayerPawnAthena MyTarget))
  (Variables|Default|SetTargetPawn _asfort_player_pawn_athena)
  (Variables|Default|SetDirtFX (Utilities|Array|Get(acopy) ParticleComponents))
  (Variables|Default|SetOrigionLocation (Transformation|GetWorldLocation (Variables|/AwfulDuck/Items/AwfulDuck/BGA_AwfulDuck.BGA_AwfulDuck_C|GetRoot (Utilities|Casting|CastToBGA_AwfulDuck _sourceobject))))
  (bind _returnvalue (Utilities|Time|SetTimerbyEvent (delegate Custom|UpdateDirtFX) 0.2 true))
  (Variables|Default|SetTimer _returnvalue))

(event Custom|UpdateDirtFX
  (bind _targetpawn (Variables|Default|GetTargetPawn))
  (bind _mesh (Variables|/Script/Engine.Character|GetMesh _targetpawn))
  (bind _returnvalue (Transformation|GetWorldLocation _mesh))
  (bind _dirtfx (Variables|Default|GetDirtFX))
  (Effects|Components|ParticleSystem|SetVectorParameter _dirtfx "DirtInitialVeclocityDirection" (Math|Vector|Normalize (- (Variables|Default|GetOrigionLocation) _returnvalue)))
  (Transformation|SetWorldLocation _dirtfx _returnvalue false false))

(event EventOnRemoval(EitherVFX-preferthisversion)
  (params (MyTarget :type Actor%20Object%20Reference) (Parameters :type Gameplay%20Cue%20Parameters%20Structure%20%28by%20ref%29) (ParticleComponents :type Array%20of%20FXSystem%20Component%20Object%20References) (AudioComponents :type Array%20of%20Audio%20Component%20Object%20References) (BurstCameraShakeInstance :type Camera%20Shake%20Base%20Object%20Reference) (BurstDecalInstance :type Decal%20Actor%20Object%20Reference))
  (Development|PrintString 0 "GCNL LocalLight - Start")
  (Utilities|Time|ClearandInvalidateTimerbyHandle MyTarget (Variables|Default|GetTimer)))

(fn ConstructionScript)
