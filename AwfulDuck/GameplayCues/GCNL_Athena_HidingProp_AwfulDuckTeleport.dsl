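; Looping gameplay cue for the AwfulDuck hiding prop teleport, handling travel audio and client-side start and end signals.
; On application and removal it sends teleport started and complete gameplay events to locally controlled pawns, so client logic can react.
; For local pawns, looping start replaces any previous travel audio with a stored 2D warp sound.
; Removal fades that sound and plays a 2D end sound without a local-control check.
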(event EventOnApplication(Cascade)
  (params (MyTarget :type Actor%20Object%20Reference) (Parameters :type Gameplay%20Cue%20Parameters%20Structure%20%28by%20ref%29) (ParticleComponents :type Array%20of%20Cascade%20Particle%20System%20Component%20Object%20References) (AudioComponents :type Array%20of%20Audio%20Component%20Object%20References) (BurstCameraShakeInstance :type Camera%20Shake%20Base%20Object%20Reference) (BurstDecalInstance :type Decal%20Actor%20Object%20Reference))
  ;[1 0.9641 0.1476] Broadcast the teleport start gameplay event to the client to trigger any client side logic that may need to occur right as this effect is added (i.e. TOD changes, hide storm actor, etc.)
  (bind _asfort_player_pawn (Utilities|Casting|CastToFortPlayerPawn MyTarget))
  (if (Pawn|IsLocallyControlled _asfort_player_pawn)
    (Ability|SendGameplayEventtoActor _asfort_player_pawn (Variables|Default|GetTeleportStartedGameplayEventTag) (Utilities|Struct|MakeGameplayEventData (Variables|Default|GetTeleportStartedGameplayEventTag) self _asfort_player_pawn))))

(event EventOnLoopingStart(Cascade)
  (params (MyTarget :type Actor%20Object%20Reference) (Parameters :type Gameplay%20Cue%20Parameters%20Structure%20%28by%20ref%29) (ParticleComponents :type Array%20of%20Cascade%20Particle%20System%20Component%20Object%20References) (AudioComponents :type Array%20of%20Audio%20Component%20Object%20References))
  (bind _asfort_player_pawn (Utilities|Casting|CastToFortPlayerPawn MyTarget)
    (:then
      (if (Pawn|IsLocallyControlled _asfort_player_pawn)
        (Utilities|FlowControl|Sequence
          (:then_0
            (bind _1ptravelaudio (Variables|Default|Get1PTravelAudio)
              (:then
                (Audio|Components|Audio|FadeOut _1ptravelaudio 0.25 0.0 "Linear" :!class /Script/Engine.AudioComponent))
              (:else)))
          (:then_1
            (bind _returnvalue (Audio|SpawnSound2D 0 (Variables|Sound|GetWarpSound)))
            (Variables|Default|Set1PTravelAudio _returnvalue)))
        (else
          (Development|PrintString 0 "Not locally controlled"))))
    (:CastFailed
      (Development|PrintString 0 "Not a fort pawn"))))

(event EventOnRemoval(Cascade)
  (params (MyTarget :type Actor%20Object%20Reference) (Parameters :type Gameplay%20Cue%20Parameters%20Structure%20%28by%20ref%29) (ParticleComponents :type Array%20of%20Cascade%20Particle%20System%20Component%20Object%20References) (AudioComponents :type Array%20of%20Audio%20Component%20Object%20References) (BurstCameraShakeInstance :type Camera%20Shake%20Base%20Object%20Reference) (BurstDecalInstance :type Decal%20Actor%20Object%20Reference))
  (Audio|PlaySound2D MyTarget (Variables|Sound|GetWarpEndSound))
  (Audio|Components|Audio|FadeOut (Variables|Default|Get1PTravelAudio) 0.5 0.0 "Linear" :!class /Script/Engine.AudioComponent)
  ;[1 0.9641 0.1476] Broadcast the teleport end gameplay event to the client to trigger any client side logic that may need to occur right as this effect is removed (i.e. TOD changes, hide storm actor, etc.)
  (bind _asfort_player_pawn (Utilities|Casting|CastToFortPlayerPawn MyTarget))
  (if (Pawn|IsLocallyControlled _asfort_player_pawn)
    (Ability|SendGameplayEventtoActor _asfort_player_pawn (Variables|Default|GetTeleportCompleteGameplayEventTag) (Utilities|Struct|MakeGameplayEventData (Variables|Default|GetTeleportCompleteGameplayEventTag) self _asfort_player_pawn))))

(event Custom|RemoveParent
  (|CallParentFunction (stale EventOnRemoval(Cascade) MyTarget) (stale EventOnRemoval(Cascade) Parameters) (stale EventOnRemoval(Cascade) ParticleComponents) (stale EventOnRemoval(Cascade) AudioComponents) (stale EventOnRemoval(Cascade) BurstCameraShakeInstance) (stale EventOnRemoval(Cascade) BurstDecalInstance)))

(fn ConstructionScript)
